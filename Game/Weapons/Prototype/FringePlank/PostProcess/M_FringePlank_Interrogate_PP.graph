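# Post-process material for the FringePlank prototype weapon's Interrogate effect: a tinted scan pulse expanding horizontally from a world origin.
# A parameter collection supplies origin, radius, timestamp and duration, driving an XY sphere-mask ring that grows, then fades late.
# The ring modulates desaturated scene colour with world-aligned grid squares and four-tap depth edge detection (deferred only), added over the scene.
# Separate tip-highlight and looping-radius branches feed no named node.
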Material /Game/Weapons/Prototype/FringePlank/PostProcess/M_FringePlank_Interrogate_PP
graph {
 properties {blendableLocation=BL_SceneColorAfterDOF, materialDomain=MD_PostProcess, pixelDepthOffsetMode=PDOM_Legacy, usedShadingModels=Default Lit}
 nodes [
  n1 = MaterialFunctionCall(
   TimestampA=CollectionParameter() {collection=&/MPC_FringePlank_Interrogate, parameterName=ActivatedTimestamp}
   TransitionLength_In=CollectionParameter() {collection=&MPC_FringePlank_Interrogate, parameterName=Duration}
   Transition In Only?=StaticBool() {value=true}
  ) {materialFunction=&MF_Timestamp_InOut_01}
  n2 = MaterialFunctionCall(
   TextureObject=TextureObject() {samplerType=SAMPLERTYPE_LinearColor, texture=/Game/Gadgets/Assets/Teleporter/Textures/T_Teleporter_GridSquares_01}
   TextureSize=ScalarParameter() {defaultValue=256, parameterName=WorldTextureScale}
   World Space Normal=ShadingPathSwitch(Default=VertexNormalWS(), Deferred=SceneTexture() {sceneTextureId=PPI_WorldNormal}.Color)
  ) {materialFunction=&WorldAlignedTexture}
  n3 = Sine(None=Add(A=@n1.Linear, B=ComponentMask(None=@n2.XYZ Texture) {r=true}))
  n4 = ComponentMask(None=CollectionParameter() {collection=&MPC_FringePlank_Interrogate, parameterName=OriginPositionWS}) {b=true, g=true, r=true}
  n5 = Constant3Vector() {constant=<rgba 1 1 0 0>}
  n6 = Multiply(A=@n4, B=@n5)
  n7 = Multiply(A=WorldPosition().XYZ, B=@n5)
  n8 = Multiply(
   A=CollectionParameter() {collection=&MPC_FringePlank_Interrogate, parameterName=Radius}
   B=@n1.Linear
  )
  n9 = Constant()
  n10 = Power(Base=OneMinus(None=SphereMask(A=@n6, B=@n7, Radius=@n8, Hardness=@n9)), Exp=ScalarParameter() {defaultValue=1.5, parameterName=Exponent})
  n11 = SphereMask(
   A=@n10
   B=ScalarParameter() {defaultValue=0.5, parameterName=RingCenter}
   Radius=ScalarParameter() {defaultValue=0.5, parameterName=RingRadius}
   Hardness=ScalarParameter() {parameterName=RingHardness}
  )
  n12 = Constant() {r=1}
  n13 = ScreenPosition()
  n14 = Multiply(B=SceneTexelSize()) {constA=3}
  n15 = Multiply(A=@n14, B=Constant2Vector() {g=-1, r=-1})
  n16 = SceneTexture() {sceneTextureId=PPI_SceneDepth}
  n17 = Subtract(A=SceneTexture(UVs=Add(A=@n13.ViewportUV, B=@n15)) {sceneTextureId=PPI_SceneDepth}.Color, B=@n16.Color)
  n18 = Multiply(A=@n14, B=Constant2Vector() {g=-1, r=1})
  n19 = Subtract(A=SceneTexture(UVs=Add(A=@n13.ViewportUV, B=@n18)) {sceneTextureId=PPI_SceneDepth}.Color, B=@n16.Color)
  n20 = Constant2Vector() {g=1, r=1}
  n21 = SceneTexture(UVs=Add(A=@n13.ViewportUV, B=Multiply(A=@n14, B=@n20))) {sceneTextureId=PPI_SceneDepth}
  n22 = Add(A=@n13.ViewportUV, B=Multiply(A=@n14, B=Constant2Vector() {g=1, r=-1}))
  n23 = Add(
   A=Add(A=Add(A=@n17, B=@n19), B=Subtract(A=@n21.Color, B=@n16.Color))
   B=Subtract(A=SceneTexture(UVs=@n22) {sceneTextureId=PPI_SceneDepth}.Color, B=@n16.Color)
  )
  n24 = Multiply(A=ComponentMask(None=Abs(None=@n23)) {r=true}, B=ScalarParameter() {defaultValue=0.01, parameterName=EdgeDetectionScaleA})
  n25 = Multiply(A=Round(None=Saturate(None=@n24)), B=ScalarParameter() {defaultValue=10, parameterName=EdgeDetectionScaleB})
  n26 = Add(A=@n11, B=Multiply(A=@n11, B=ShadingPathSwitch(Default=@n12, Deferred=@n25)))
  n27 = ComponentMask(None=SceneTexture() {sceneTextureId=PPI_PostProcessInput0}.Color) {b=true, g=true, r=true}
  n28 = Desaturation(None=@n27, Fraction=ScalarParameter() {defaultValue=0.5, parameterName=DesaturationFraction})
  n29 = ComponentMask(None=@n2.XYZ Texture) {g=true}
  n30 = Add(A=Multiply(A=ConstantBiasScale(None=@n3) {bias=2}, B=Multiply(A=@n26, B=@n28)), B=Multiply(A=@n11, B=OneMinus(None=@n29)))
  n31 = VectorParameter() {parameterName=Color, defaultValue=<rgba 0.0288 0.173 0.5 0>}
  n32 = Multiply(A=OneMinus(None=@n1.Linear)) {constB=4}
  Add(A=Multiply(A=Multiply(A=@n30, B=@n31.RGB), B=Saturate(None=@n32)), B=@n27)
  n33 = MaterialFunctionCall(Float3=WorldPosition().XYZ) {materialFunction=&BreakOutFloat3Components}
  n34 = ScalarParameter() {defaultValue=6000, parameterName=HeightMaskRadius}
  n35 = ScalarParameter() {defaultValue=0.25, parameterName=HeightMaskHardness}
  n36 = ScalarParameter() {defaultValue=30, parameterName=TipScale}
  n37 = Power(
   Base=SphereMask(
    A=@n10
    B=ScalarParameter() {defaultValue=0.93, parameterName=TipCenter}
    Radius=ScalarParameter() {defaultValue=0.07, parameterName=TipRadius}
    Hardness=ScalarParameter() {parameterName=TipHardness}
   )
   Exp=ScalarParameter() {defaultValue=3, parameterName=ExponentTip}
  )
  Add(A=Multiply(A=SphereMask(A=@n4, B=@n33.B, Radius=@n34, Hardness=@n35), B=Multiply(A=@n36, B=@n37)), B=@n30)
  n38 = Time()
  n39 = ScalarParameter() {defaultValue=0.6666, parameterName=TimeScale}
  Multiply(A=ScalarParameter() {defaultValue=9000, parameterName=SourceRadius}, B=Fmod(A=Multiply(A=@n38, B=@n39), B=Constant() {r=1}))
 ]
}
