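# Post-process material for the FringePlank prototype weapon's Interrogate effect, drawing an expanding world-space scan over the scene before tonemapping.
# An MPC supplies origin, radius, activation timestamp and duration, driving a growing radius that shapes a scanline wave and trailing gradient with height falloff.
# World-aligned animated grid, four-tap depth edge detection and custom-depth masks feed coloured patterns.
# Output compensates for eye adaptation; shading-path switches provide non-deferred fallbacks.
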Material /Game/Weapons/Prototype/FringePlank/PostProcess/M_FringePlank_Interrogate_PP_2025
graph {
 properties {blendableLocation=BL_SceneColorAfterDOF, materialDomain=MD_PostProcess, pixelDepthOffsetMode=PDOM_Legacy, usedShadingModels=Default Lit}
 nodes [
  n1 = Multiply(A=ScalarParameter() {defaultValue=1, parameterName=Overall_Intensity}, B=Add(A=NamedRerouteUsage(), B=NamedRerouteUsage()))
  n2 = NamedRerouteUsage()
  n3 = Saturate(None=Multiply(A=OneMinus(None=@n2)) {constB=5})
  n4 = ComponentMask(None=SceneTexture() {sceneTextureId=PPI_PostProcessInput0}.Color) {b=true, g=true, r=true}
  Add(A=Divide(A=Multiply(A=@n1, B=@n3), B=EyeAdaptation().EyeAdaptation), B=@n4)
  n5 = NamedRerouteUsage()
  n6 = MaterialFunctionCall(Float3=WorldPosition().XYZ) {materialFunction=&BreakOutFloat3Components}
  n7 = ScalarParameter() {defaultValue=6000, parameterName=HeightMaskRadius}
  n8 = ScalarParameter() {defaultValue=0.25, parameterName=HeightMaskHardness}
  n9 = ScalarParameter() {defaultValue=30, parameterName=TipScale}
  n10 = Power(
   Base=SphereMask(
    B=ScalarParameter() {defaultValue=0.93, parameterName=TipCenter}
    Radius=ScalarParameter() {defaultValue=0.07, parameterName=TipRadius}
    Hardness=ScalarParameter() {parameterName=TipHardness}
   )
   Exp=ScalarParameter() {defaultValue=3, parameterName=ExponentTip}
  )
  Add(A=Multiply(A=SphereMask(A=@n5, B=@n6.B, Radius=@n7, Hardness=@n8), B=Multiply(A=@n9, B=@n10)))
  n11 = Time()
  n12 = TextureObjectParameter() {parameterName=GridTexture, samplerType=SAMPLERTYPE_LinearColor, texture=/Game/Gadgets/Assets/Teleporter/Textures/T_Teleporter_GridSquares_02}
  n13 = MaterialFunctionCall(
   TextureObject=@n12
   TextureSize=ScalarParameter() {defaultValue=256, parameterName=WorldTextureScale}
   World Space Normal=ShadingPathSwitch(Default=VertexNormalWS(), Deferred=SceneTexture() {sceneTextureId=PPI_WorldNormal}.Color)
  ) {materialFunction=&WorldAlignedTexture}
  n14 = ComponentMask(None=@n13.XYZ Texture) {b=true}
  n15 = Multiply(A=@n11) {constB=0.1}
  n16 = Power(Base=NamedRerouteUsage()) {constExponent=0.2}
  n17 = Sine(None=Add(A=Add(A=@n15, B=@n16), B=@n14)) {period=0.5}
  n18 = LinearInterpolate(A=Sine(None=Add(A=@n11, B=@n14)) {period=0.15}, Alpha=@n17)
  n19 = Add(A=OneMinus(None=NamedRerouteUsage()), B=@n17)
  n20 = Multiply(A=Multiply(A=Saturate(None=@n18), B=Saturate(None=@n19)), B=ComponentMask(None=@n13.XYZ Texture) {g=true})
  n21 = LinearInterpolate(Alpha=NamedRerouteUsage()) {constA=0.6, constB=5}
  n22 = ComponentMask(None=@n13.XYZ Texture) {r=true}
  n23 = Multiply(
   A=ScalarParameter() {defaultValue=1, parameterName=Grid_IntensityHigh}
   B=StaticSwitchParameter(True=Power(Base=@n20, Exp=@n21), False=OneMinus(None=@n22)) {parameterName=UseRuneGrid}
  ) {constB=0.2}
  n24 = ComponentMask(None=@n13.XYZ Texture) {g=true}
  Add(A=@n23, B=Multiply(A=OneMinus(None=@n24), B=ScalarParameter() {defaultValue=1, parameterName=Grid_IntensityLow}) {constB=0.2})
  Constant()
  n25 = Time()
  n26 = ScalarParameter() {defaultValue=0.6666, parameterName=TimeScale}
  Multiply(A=ScalarParameter() {defaultValue=9000, parameterName=SourceRadius}, B=Fmod(A=Multiply(A=@n25, B=@n26), B=Constant() {r=1}))
  n27 = ScreenPosition()
  n28 = Multiply(B=SceneTexelSize()) {constA=3}
  n29 = Add(A=@n27.ViewportUV, B=Multiply(A=@n28, B=Constant2Vector() {g=-1, r=-1}))
  n30 = SceneTexture() {sceneTextureId=PPI_SceneDepth}
  n31 = Add(A=@n27.ViewportUV, B=Multiply(A=@n28, B=Constant2Vector() {g=-1, r=1}))
  n32 = Add(
   A=Subtract(A=SceneTexture(UVs=@n29) {sceneTextureId=PPI_SceneDepth}.Color, B=@n30.Color)
   B=Subtract(A=SceneTexture(UVs=@n31) {sceneTextureId=PPI_SceneDepth}.Color, B=@n30.Color)
  )
  n33 = Multiply(A=@n28, B=Constant2Vector() {g=1, r=1})
  n34 = Subtract(A=SceneTexture(UVs=Add(A=@n27.ViewportUV, B=@n33)) {sceneTextureId=PPI_SceneDepth}.Color, B=@n30.Color)
  n35 = Constant2Vector() {g=1, r=-1}
  n36 = SceneTexture(UVs=Add(A=@n27.ViewportUV, B=Multiply(A=@n28, B=@n35))) {sceneTextureId=PPI_SceneDepth}
  n37 = Abs(None=Add(A=Add(A=@n32, B=@n34), B=Subtract(A=@n36.Color, B=@n30.Color)))
  n38 = Saturate(None=Multiply(A=ComponentMask(None=@n37) {r=true}, B=ScalarParameter() {defaultValue=0.01, parameterName=EdgeDetectionScaleA}))
  n39 = SmoothStep(Value=@n38) {constMin=0.3}
  n40 = ScalarParameter() {defaultValue=10, parameterName=EdgeDetectionScaleB}
  NamedRerouteDeclaration(None=ShadingPathSwitch(Default=Constant() {r=1}, Deferred=Multiply(A=@n39, B=@n40))) {name=EdgeDetection, nodeColor=<rgba 0.8516 0 1 1>}
  n41 = ComponentMask(None=CollectionParameter() {collection=&/MPC_FringePlank_Interrogate, parameterName=OriginPositionWS}) {b=true, g=true, r=true}
  NamedRerouteDeclaration(None=@n41) {name=WorldPosOG, nodeColor=<rgba 0 0.1094 1 1>}
  n42 = VectorParameter() {parameterName=Color02, defaultValue=<rgba 0.0299 0.0015 0.8844 1>}
  n43 = NamedRerouteUsage()
  n44 = Multiply(A=NamedRerouteUsage(), B=@n43)
  n45 = NamedRerouteUsage()
  n46 = NamedRerouteUsage()
  n47 = ScalarParameter() {defaultValue=0.2, parameterName=GrowingIntensity}
  n48 = Add(
   A=Multiply(A=Multiply(A=@n44, B=@n45), B=ScalarParameter() {defaultValue=2, parameterName=GridIntensity})
   B=Multiply(A=NamedRerouteUsage(), B=Multiply(A=@n46, B=@n47))
  )
  n49 = Multiply(A=@n43, B=NamedRerouteUsage())
  n50 = Multiply(A=Multiply(A=Add(A=@n43, B=@n49), B=NamedRerouteUsage()), B=NamedRerouteUsage())
  n51 = NamedRerouteDeclaration(None=VectorParameter() {parameterName=Color01, defaultValue=<rgba 0.0168 0.1412 0.3177 1>}.RGB) {name=Color01, nodeColor=<rgba 0.3359 1 0 1>}
  NamedRerouteDeclaration(None=Add(A=Multiply(A=@n42.RGB, B=@n48), B=Multiply(A=@n50, B=@n51))) {name=GradientPattern, nodeColor=<rgba 1 0 0.5625 1>}
  n52 = VectorParameter() {parameterName=WorldPositionMultiply, defaultValue=<rgba 1 1 0 1>}
  n53 = Multiply(A=@n41, B=@n52.RGB)
  n54 = WorldPosition()
  n55 = Multiply(A=@n54.XYZ, B=@n52.RGB)
  n56 = Multiply(A=NamedRerouteUsage(), B=NamedRerouteUsage())
  n57 = Constant()
  n58 = Saturate(None=OneMinus(None=SphereMask(A=@n53, B=@n55, Radius=@n56, Hardness=@n57)))
  n59 = ScalarParameter() {defaultValue=1, parameterName=Exponent}
  n60 = NamedRerouteUsage()
  n61 = ScalarParameter() {defaultValue=500, parameterName=RadiusOffset}
  n62 = Saturate(None=SphereMask(A=@n53, B=@n55, Radius=Add(A=@n60, B=@n61), Hardness=@n57))
  NamedRerouteDeclaration(None=Multiply(A=Power(Base=@n58, Exp=@n59), B=Power(Base=@n62, Exp=@n59))) {name=GrowingGradientOG, nodeColor=<rgba 1 0 0.4453 1>}
  n63 = ComponentMask(None=@n41) {b=true}
  n64 = ComponentMask(None=@n54.XYZ) {b=true}
  NamedRerouteDeclaration(None=Abs(None=Subtract(A=@n63, B=@n64))) {name=Z_Distance, nodeColor=<rgba 1 0 0.1641 1>}
  n65 = Saturate(None=Divide(A=NamedRerouteUsage(), B=ScalarParameter() {defaultValue=500, parameterName=Z_GradientUnit}))
  n66 = LinearInterpolate(Alpha=Power(Base=OneMinus(None=@n65))) {constA=0.1}
  NamedRerouteDeclaration(None=@n66) {name=Z_Gradient_5M, nodeColor=<rgba 0 0.0625 1 1>}
  NamedRerouteDeclaration(None=@n23) {name=GridMask_Detail, nodeColor=<rgba 1 0 0.75 1>}
  NamedRerouteDeclaration(None=@n23) {name=GridMask, nodeColor=<rgba 1 0 0.75 1>}
  NamedRerouteDeclaration(None=Distance(A=@n53, B=@n55)) {name=DistanceWS, nodeColor=<rgba 1 0 0.3516 1>}
  n67 = SceneTexture() {sceneTextureId=PPI_CustomDepth}
  n68 = Saturate(None=Step(X=ComponentMask(None=@n67.Color) {r=true}) {constX=100000, constY=10000000})
  NamedRerouteDeclaration(None=@n68) {name=GridMask_Stencil, nodeColor=<rgba 0 1 0.9531 1>}
  NamedRerouteDeclaration(None=Desaturation(None=@n4, Fraction=ScalarParameter() {defaultValue=0.5, parameterName=DesaturationFraction})) {name=GrayScaleScene, nodeColor=<rgba 1 0 0.9141 1>}
  n69 = CollectionParameter() {collection=&MPC_FringePlank_Interrogate, parameterName=ActivatedTimestamp}
  n70 = CollectionParameter() {collection=&MPC_FringePlank_Interrogate, parameterName=Duration}
  n71 = StaticBool() {value=true}
  n72 = Multiply(
   A=NamedRerouteDeclaration(None=CollectionParameter() {collection=&MPC_FringePlank_Interrogate, parameterName=Radius}) {name=RadiusTotal, nodeColor=<rgba 0 0.3672 1 1>}
   B=NamedRerouteDeclaration(None=MaterialFunctionCall(TimestampA=@n69, TransitionLength_In=@n70, Transition In Only?=@n71) {materialFunction=&MF_Timestamp_InOut_01}.Linear) {name=ActiveAlpha, nodeColor=<rgba 0.4297 1 0 1>}
  )
  n73 = Subtract(A=NamedRerouteUsage(), B=@n72)
  n74 = Saturate(None=Divide(A=@n73, B=ScalarParameter() {defaultValue=400, parameterName=ScanlineWidth}))
  n75 = OneMinus(None=@n74)
  n76 = ScalarParameter() {defaultValue=5000, parameterName=GradientWidth}
  n77 = Saturate(None=Subtract(A=@n72, B=@n76))
  n78 = Saturate(None=Divide(A=Subtract(A=@n73, B=@n77), B=@n76))
  NamedRerouteDeclaration(None=Multiply(A=NamedRerouteDeclaration(None=@n75) {name=MaxArea, nodeColor=<rgba 1 0.4219 0 1>}, B=OneMinus(None=@n78))) {name=GradientArea, nodeColor=<rgba 1 0 0.3984 1>}
  n79 = Multiply(A=@n74, B=@n75)
  n80 = Multiply(A=@n66, B=Multiply(A=@n79, B=ScalarParameter() {defaultValue=0.4, parameterName=EM_IntensityWave}))
  n81 = NamedRerouteUsage()
  n82 = Multiply(A=ScalarParameter() {defaultValue=2, parameterName=EM_IntensityWave_Grid}, B=Multiply(A=Multiply(A=@n79, B=@n81), B=NamedRerouteUsage()))
  NamedRerouteDeclaration(None=Multiply(A=VectorParameter() {parameterName=Color03, defaultValue=<rgba 0.1474 0.7542 1 1>}.RGB, B=Add(A=@n80, B=@n82))) {name=WavePattern, nodeColor=<rgba 0 0.7187 1 1>}
  NamedRerouteDeclaration(None=@n79) {name=DebugWave, nodeColor=<rgba 1 0 0.5156 1>}
  NamedRerouteUsage()
  NamedRerouteUsage()
  NamedRerouteUsage()
  n83 = Fmod(A=Time(), B=Constant() {r=3})
  Reroute(None=Saturate(None=Divide(A=@n83)))
  Round(None=@n38)
  n84 = Divide(A=NamedRerouteUsage(), B=ScalarParameter() {defaultValue=200, parameterName=CoreUnit})
  Saturate(None=OneMinus(None=Saturate(None=@n84)))
  ScalarParameter() {defaultValue=1, parameterName=Exponent}
  ShadingPathSwitch(Default=Constant() {r=1}, Deferred=OneMinus(None=@n68))
  Step() {constX=100000}
 ]
}
